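# One-shot stereo MetaSound source for Battle Royale boss NPC voice lines, with an empty default wave array.
# A probability gate starts a random delay, then a shuffled wave plays with random pitch and volume.
# A failed roll or wave completion fires OnFinished, so skipped plays still end.
# It limits concurrency to one per owner, preventing new plays, and pre-effect sends to a boss VO envelope-follower bus.
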MetaSoundSource /Game/Sounds/Fort_Human_VO_BR/NPCs/Shared/MSS_VO_BR_Boss_Stereo
inputs [
 "Delay Max (Seconds)" {type=Time}
 "Delay Min (Seconds)" {type=Time}
 "Probability (0-1)" {type=Float}
 "Random Pitch Max (Semitones)" {type=Float}
 "Random Pitch Min (Semitones)" {type=Float}
 "Random Volume Change Max (Decibels)" {type=Float}
 "Random Volume Change Min (Decibels)" {type=Float}
 UE.Source.OnPlay {type=Trigger}
 Wave Array {type=WaveAsset:Array}
]
interfaces [UE.OutputFormat.Stereo:1.0, UE.Source.OneShot:1.0, UE.Source:1.1]
outputs [
 "UE.OutputFormat.Stereo.Audio:0" {type=Audio}
 "UE.OutputFormat.Stereo.Audio:1" {type=Audio}
 UE.Source.OneShot.OnFinished {type=Trigger}
]
pages [Default {
  input_defaults {
   "Delay Max (Seconds)"=0
   "Delay Min (Seconds)"=0
   "Probability (0-1)"=1
   "Random Pitch Max (Semitones)"=0
   "Random Pitch Min (Semitones)"=0
   "Random Volume Change Max (Decibels)"=0
   "Random Volume Change Min (Decibels)"=0
   UE.Source.OnPlay=null
   Wave Array []
  }
  nodes [
   TriggerFilter_0 = UE.Trigger Filter(Probability=^"Probability (0-1)", Trigger=^UE.Source.OnPlay)
   n1 = UE.RandomTime(Max=^"Delay Max (Seconds)", Min=^"Delay Min (Seconds)", Next=@TriggerFilter_0.Heads)
   TriggerDelay_0 = UE.Trigger Delay(In=@n1.On Next, Delay Time=@n1.Value)
   n2 = Array.Shuffle:WaveAsset:Array(In Array=^Wave Array, Next=@TriggerDelay_0.Out) {also {defaults {Enable Shared State="<MetaSoundBuilderResult.SUCCEEDED: 0>"}}}
   n3 = UE.RandomFloat(Max=^"Random Pitch Max (Semitones)", Min=^"Random Pitch Min (Semitones)", Next=@n2.On Next)
   n4 = UE.RandomFloat(Max=^"Random Volume Change Max (Decibels)", Min=^"Random Volume Change Min (Decibels)", Next=@n3.On Next)
   DecibelstoLinearGain_0 = UE.Decibels to Linear Gain:Float(Decibels=@n4.Value)
   WavePlayer_0 = UE.Wave Player:Stereo(Wave Asset=@n2.Value, Play=@n4.On Next, Pitch Shift=UE.Reroute(Value=@n3.Value).Value) {also {version [1,1]}}
   Multiply_0 = UE.Multiply:Audio by Float(AdditionalOperands=@DecibelstoLinearGain_0.Linear Gain, PrimaryOperand=@WavePlayer_0.Out Left)
   Multiply_1 = UE.Multiply:Audio by Float(AdditionalOperands=@DecibelstoLinearGain_0.Linear Gain, PrimaryOperand=@WavePlayer_0.Out Right)
   n5 = InitVariable.Trigger()
   TriggerAnyTriggerAny2_0 = "TriggerAny.Trigger Any (2)"(
    In 1=VariableAccessor.Trigger(Variable=VariableMutator.Trigger(Variable=@n5.Variable, Value=@TriggerFilter_0.Tails).Variable).Value
    In 0=@WavePlayer_0.On Finished
   )
  ]
  links [
   {dst="@UE.OutputFormat.Stereo.Audio:0", src=Multiply_0.Out}
   {dst="@UE.OutputFormat.Stereo.Audio:1", src=Multiply_1.Out}
   {dst="@UE.Source.OneShot.OnFinished", src=TriggerAnyTriggerAny2_0.Out}
  ]
 }]
properties {
 assetImportData=&self:AssetImportData
 attenuationSettings=&/NPC_VO_Alerted_Att
 maxDistance=12500
 numChannels=2
 outputFormat=Stereo
 sampleRate=48000
 soundClassObject=&Fort_VO_Enemy_Critical
 concurrencyOverrides {bLimitToOwner=true, maxCount=1, resolutionRule=PreventNew}
 concurrencySet [&NPC_Boss_Combat_Chatter_Conc]
 preEffectBusSends [{
   audioBus=/Game/Sounds/Modulation/EnvelopeFollowers/BR_Boss/AB_BR_Boss_VO_EnvFollow
   maxSendDistance=8000
   minSendDistance=0
   minSendLevel=1
   sendLevel=1
  }]
}
